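; BP_LimeRock_Portal is a portal Blueprint that publishes its placement to the MPC_LimeRock material parameter collection.
; Its construction script passes PortalRadius to SetPortalMPCParams, which reads the NS_LimeRock_RealityTear_Portal component's world transform.
; The function writes location and radius packed into PortalPositionRadius.
; It writes the forward vector to PortalDirectionIntensity, first rounding rotation to whole degrees.
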(fn SetPortalMPCParams
  (params (Radius :type Float%20%28single-precision%29))
  (bind _ns_limerock_realitytear_portal (Variables|Default|GetNS_LimeRock_RealityTear_Portal))
  (bind (_returnvalue_roll _returnvalue_pitch _returnvalue_yaw) (Math|Rotator|BreakRotator (Transformation|GetWorldRotation _ns_limerock_realitytear_portal)))
  (bind (_returnvalue_x _returnvalue_y _returnvalue_z) (Math|Vector|BreakVector (Transformation|GetWorldLocation _ns_limerock_realitytear_portal)))
  (Rendering|Material|SetVectorParameterValue "/LimeRock/Assets/Materials/MPC_LimeRock.MPC_LimeRock" "PortalPositionRadius" (Math|Color|MakeColor _returnvalue_x _returnvalue_y _returnvalue_z Radius) :!class /Script/Engine.KismetMaterialLibrary)
  (Rendering|Material|SetVectorParameterValue "/LimeRock/Assets/Materials/MPC_LimeRock.MPC_LimeRock" "PortalDirectionIntensity" (Math|Conversions|ToLinearColor(Vector) (Math|Vector|GetForwardVector (Math|Rotator|MakeRotator (Math|Conversions|ToFloat(Integer) (Math|Float|Round _returnvalue_roll)) (Math|Conversions|ToFloat(Integer) (Math|Float|Round _returnvalue_pitch)) (Math|Conversions|ToFloat(Integer) (Math|Float|Round _returnvalue_yaw))))) :!class /Script/Engine.KismetMaterialLibrary))

(fn ConstructionScript
  (CallFunction|SetPortalMPCParams (Variables|Default|GetPortalRadius)))
