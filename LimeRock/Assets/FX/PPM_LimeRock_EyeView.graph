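# Post-process material for an eye-view screen effect that distorts and tints the scene and adds animated colored veins.
# Four offset scene samples, jittered by temporal dither, average into a radial blur blended with a desaturated tinted copy; named reroutes define center-to-edge mask, Opacity and Pulse controls.
# Veins multiply two panning plasma textures in radial coordinates under a radial mask.
# A default-on static switch skips the blur and tint on forward and mobile paths.
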Material /LimeRock/Assets/FX/PPM_LimeRock_EyeView
graph {
 properties {bAutomaticallySetUsageInEditor=false, materialDomain=MD_PostProcess, usedShadingModels=Default Lit}
 nodes [
  n1 = VectorParameter() {parameterName=VeinColor, defaultValue=<rgba 1.5 0.3449 0.3592 1>}
  n2 = NamedRerouteUsage()
  n3 = SceneTexture() {sceneTextureId=PPI_PostProcessInput0}
  n4 = ScreenPosition()
  n5 = Constant2Vector() {g=0.5, r=0.5}
  n6 = Subtract(A=@n4.ViewportUV, B=@n5)
  n7 = NamedRerouteUsage()
  n8 = Multiply(A=@n6, B=Max(A=ComponentMask(None=@n7) {g=true, r=true}, B=Constant() {r=0.5}))
  n9 = LinearInterpolate(B=ScalarParameter() {defaultValue=-0.04, parameterName=OffsetAmount}, Alpha=NamedRerouteUsage())
  n10 = Multiply(A=@n9)
  n11 = MaterialFunctionCall() {materialFunction=&DitherTemporalAA}
  n12 = AppendVector(A=@n6, B=Constant())
  n13 = Constant3Vector()
  n14 = Multiply(
   A=ComponentMask(None=MaterialFunctionCall(Vector=@n12, Default=@n13) {materialFunction=&SafeNormalize}.Result) {g=true, r=true}
   B=LinearInterpolate(B=ScalarParameter() {defaultValue=0.025, parameterName=BlurDither}, Alpha=NamedRerouteUsage())
  )
  n15 = MaterialFunctionCall(Float2=Add(A=Multiply(A=@n11.Result, B=@n14), B=@n4.ViewportUV)) {materialFunction=&BreakOutFloat2Components}
  n16 = Multiply(A=MaterialFunctionCall(X=Clamp(None=@n15.R), Y=Clamp(None=@n15.G)) {materialFunction=&MakeFloat2}.Result)
  n17 = SceneTexture(UVs=Add(A=Multiply(A=@n8, B=@n10), B=@n16)) {sceneTextureId=PPI_PostProcessInput0}
  n18 = Multiply(A=@n9) {constB=2}
  n19 = SceneTexture(UVs=Add(A=Multiply(A=@n8, B=@n18), B=@n16)) {sceneTextureId=PPI_PostProcessInput0}
  n20 = Multiply(A=@n9) {constB=3}
  n21 = SceneTexture(UVs=Add(A=Multiply(A=@n8, B=@n20), B=@n16)) {sceneTextureId=PPI_PostProcessInput0}
  n22 = Multiply(A=@n9) {constB=4}
  n23 = SceneTexture(UVs=Add(A=Multiply(A=@n8, B=@n22), B=@n16)) {sceneTextureId=PPI_PostProcessInput0}
  n24 = Divide(A=Add(A=Add(A=@n17.Color, B=@n19.Color), B=Add(A=@n21.Color, B=@n23.Color))) {constB=4}
  n25 = Multiply(
   A=Desaturation(None=@n24, Fraction=ScalarParameter() {defaultValue=0.5, parameterName=Desaturation})
   B=VectorParameter() {parameterName=Tint, defaultValue=<rgba 0.7866 0.1921 0.8 1>}.RGB
  ) {constA=1}
  n26 = Constant() {r=1}
  n27 = Multiply(A=NamedRerouteUsage()) {constB=3}
  n28 = LinearInterpolate(A=@n3.Color, B=LinearInterpolate(A=@n24, B=AppendVector(A=@n25, B=@n26), Alpha=Saturate(None=@n27)), Alpha=NamedRerouteUsage())
  Add(
   A=AppendVector(A=Multiply(A=@n1.RGB, B=@n2), B=Constant())
   B=StaticSwitchParameter(True=ShadingPathSwitch(Default=@n28, Deferred=@n28, Forward=@n3.Color, Mobile=@n3.Color), False=@n28) {defaultValue=true, parameterName=Not use on medium and low}
  )
  NamedRerouteDeclaration(None=ScalarParameter() {defaultValue=1, parameterName=Opacity}) {name=Opacity, nodeColor=<rgba 1 0.8203 0 1>}
  NamedRerouteDeclaration(None=ScalarParameter() {defaultValue=1, parameterName=Pulse}) {name=Pulse, nodeColor=<rgba 0.6641 1 0 1>}
  n29 = MaterialFunctionCall!(Vector or UVs=ScreenPosition().ViewportUV) {materialFunction=/FortUILibrary/MaterialFunctions/Basics/MF_UI_CheaperVectorToRadial}
  n30 = Power(Base=Distance(A=Constant(), B=ComponentMask(None=@n29.Radial Coordinates) {g=true})) {constExponent=2.5}
  n31 = VectorParameter() {parameterName=VeinTilePan, channelNames {a=AnimScale, b=AnimSpeed, g=Y Tile, r=X Tile}, defaultValue=<rgba 6 0.8 0.84 0.01>}
  n32 = Multiply(A=@n29.Radial Coordinates, B=AppendVector(A=@n31.X Tile, B=@n31.X Tile))
  n33 = Time()
  n34 = Multiply(A=@n31.AnimSpeed, B=@n33)
  n35 = Constant() {r=1}
  n36 = ScalarParameter() {defaultValue=0.64, parameterName=VeinAnimPower}
  n37 = Sine(None=MaterialFunctionCall(In=Fmod(A=@n34, B=@n35), Power=@n36) {materialFunction=&SCurve}.Result)
  n38 = AppendVector(A=Multiply(A=@n31.AnimScale) {constB=0.3}, B=@n31.AnimScale)
  n39 = Add(A=@n32) {constB=0.33}
  n40 = Multiply(A=@n31.AnimSpeed, B=Add(A=@n33) {constB=0.12})
  n41 = Sine(None=MaterialFunctionCall(In=Fmod(A=@n40, B=@n35), Power=@n36) {materialFunction=&SCurve}.Result)
  n42 = AppendVector(A=Multiply(A=@n31.AnimScale) {constB=-0.3}, B=Multiply(A=@n31.AnimScale) {constB=1.5})
  n43 = Multiply(
   A=TextureSample(UVs=Panner(Coordinate=@n32, Time=@n37, Speed=@n38)) {samplerType=SAMPLERTYPE_LinearColor, texture=/Game/Effects/Fort_Effects/Emotes/UltraLight/Texture/T_Plasma_A_C}.RGB
   B=TextureSample(UVs=Panner(Coordinate=@n39, Time=@n41, Speed=@n42)) {samplerType=SAMPLERTYPE_LinearColor, texture=&T_Plasma_A_C}.RGB
  )
  n44 = Constant() {r=-1}
  NamedRerouteDeclaration(None=Multiply(A=NamedRerouteDeclaration(None=@n30) {name=VeinsMask, nodeColor=<rgba 1 0 0.7734 1>}, B=Desaturation(None=@n43, Fraction=@n44))) {name=VeinTexture, nodeColor=<rgba 1 0 0.7734 1>}
  n45 = SphereMask(
   A=@n5
   B=@n4.ViewportUV
   Radius=ScalarParameter() {defaultValue=1.25, parameterName=CenterMaskSize}
   Hardness=ScalarParameter() {parameterName=CenterMaskHardness}
  ) {hardnessPercent=0}
  n46 = ScalarParameter() {defaultValue=0.7, parameterName=MaskPower}
  NamedRerouteDeclaration(None=LinearInterpolate(
    A=ScalarParameter() {defaultValue=1, parameterName=BlurEdges}
    B=ScalarParameter() {parameterName=BlurCenter}
    Alpha=Power(Base=@n45, Exp=@n46)
   )) {name=SphereMask, nodeColor=<rgba 1 0.4688 0 1>}
 ]
}
