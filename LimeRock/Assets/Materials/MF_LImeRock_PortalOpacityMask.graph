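# Material function that computes LimeRock portal opacity and interior masks for use in materials.
# It feeds portal position, direction and radius from MPC_LimeRock into the shared MF_FX_Circular_Portal_Mask function.
# The MaskGradientSize parameter, capped at the portal radius, drives edge gradient size and offset.
# A MaskEnabled static switch, on by default, selects those masks or falls back to full opacity and an empty interior.
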MaterialFunction /LimeRock/Assets/Materials/MF_LImeRock_PortalOpacityMask
graph {nodes [
  n1 = CollectionParameter() {collection=&/MPC_LimeRock, parameterName=PortalRadius}
  n2 = Min(A=@n1, B=ScalarParameter() {defaultValue=1500, parameterName=MaskGradientSize})
  n3 = MaterialFunctionCall!(
   MaskPosition=ComponentMask(None=CollectionParameter() {collection=&MPC_LimeRock, parameterName=PortalPositionRadius}) {b=true, g=true, r=true}
   MaskDirection=ComponentMask(None=CollectionParameter() {collection=&MPC_LimeRock, parameterName=PortalDirectionIntensity}) {b=true, g=true, r=true}
   MaskRadius=@n1
   MaskEdgeGradientSize=@n2
   MaskEdgeGradientOffset=Multiply(A=@n2)
   MaskEdgeGradientPower
  ) {materialFunction=/Game/Effects/Niagara/MasterMaterials/Functions/Opacity/MF_FX_Circular_Portal_Mask}
  n4 = FunctionInput(Preview=StaticBool() {value=true}) {bUsePreviewValueAsDefault=true, inputName=MaskEnabled, inputType=FunctionInput_StaticBool}
  FunctionOutput(None=StaticSwitch(True=Saturate(None=@n3.OpacityMask), False=Constant() {r=1}, Value=@n4)) {outputName=OpacityMask}
  FunctionOutput(None=StaticSwitch(True=@n3.InteriorMask, False=Constant(), Value=@n4)) {outputName=InteriorMask, sortPriority=1}
 ]}
