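# Animated tentacle eye material for skeletal meshes, built on MetaHuman eye functions.
# UVs, offset oppositely by local X, feed refraction with noise distortion, then pupil scale/shift and parallax-occlusion iris sampling, blended over rotated sclera with optional color correction.
# A time-driven flicker curve thresholds a vertical gradient to blend in a dark eyelid layer as a blink, with a forced-close parameter.
# LimeRock world-glow and portal opacity-mask functions finish the output.
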Material /LimeRock/Assets/Materials/M_Tentacle_Eye
graph {
 properties {
  bAutomaticallySetUsageInEditor=false
  bUseMaterialAttributes=true
  bUsedWithMeshDeformer=true
  bUsedWithSkeletalMesh=true
  opacityMaskClipValue=0.5
  shadingModel=MSM_FromMaterialExpression
  subsurfaceProfile=/Game/MetaHumans/Common/Face/SubsurfaceProfiles/SSP_Eye_Inner
  usedShadingModels=Default Lit
 }
 nodes [
  LinearInterpolate()
  n1 = NamedRerouteUsage()
  n2 = Add(A=Multiply(A=ScalarParameter() {defaultValue=2.1884, group=Iris, parameterName=IrisUVRadius}) {constB=0.03}) {constA=0.15, constB=0.15}
  n3 = Divide(A=Subtract(A=@n1) {constB=0.5}, B=@n2)
  n4 = Constant()
  n5 = Constant() {r=1}
  n6 = Constant() {r=0.49}
  n7 = Saturate(None=OneMinus(None=SphereMask(A=@n3, B=@n4, Radius=@n5, Hardness=@n6) {attenuationRadius=1, hardnessPercent=30}))
  n8 = ScalarParameter() {group=Iris, parameterName=DistortStrength}
  n9 = ScalarParameter() {group=Iris, parameterName=DistortPanX}
  n10 = ScalarParameter() {group=Iris, parameterName=DistortPanY}
  n11 = Add(
   A=Divide(A=@n1, B=ScalarParameter() {defaultValue=5, group=Iris, parameterName=DistortScale})
   B=Divide(A=AppendVector(A=@n9, B=@n10)) {constB=100}
  )
  n12 = Multiply(
   A=Subtract(A=@n1) {constB=0.5}
   B=LinearInterpolate(
    A=@n8
    B=Multiply(A=@n8) {constB=-1}
    Alpha=TextureSample(UVs=@n11) {samplerSource=SSM_Wrap_WorldGroupSettings, texture=&/T_TilingNoise_002}.R
   )
  )
  n13 = TextureSample(UVs=MaterialFunctionCall(X=Add(A=@n2) {constB=0.5}, Y=Constant() {r=0.5}) {materialFunction=&MakeFloat2}.Result) {
   samplerSource=SSM_Wrap_WorldGroupSettings
   samplerType=SAMPLERTYPE_LinearColor
   texture=&/T_EyeMidPlaneDisplacement1
  }
  n14 = TextureSample(UVs=NamedRerouteUsage()) {
   samplerSource=SSM_Wrap_WorldGroupSettings
   samplerType=SAMPLERTYPE_LinearColor
   texture=&T_EyeMidPlaneDisplacement1
  }
  n15 = Constant3Vector() {constant=<rgba 0 0 1 0>}
  n16 = Transform(None=@n15)
  n17 = MaterialFunctionCall!(
   InternalIoR=ScalarParameter() {defaultValue=1.28, group=Refraction, parameterName=IoR}
   UV=Add(A=NamedRerouteUsage(), B=Multiply(A=@n7, B=@n12))
   LimbusUVWidth=MaterialFunctionCall(
    X=ScalarParameter() {defaultValue=0.04, group=Iris, parameterName=IrisBorderWidth}
    Y=ScalarParameter() {defaultValue=0.035, group=Iris, parameterName=IrisBleedWidth}
    Z=ScalarParameter() {defaultValue=0.045, group=Iris, parameterName=IrisMaskWidth}
   ) {materialFunction=&MakeFloat3}.Result
   DepthScale=ScalarParameter() {defaultValue=0.8735, group=Refraction, parameterName=RefractionDepthScale}
   DepthPlaneOffset=@n13.R
   MidPlaneDisplacement=@n14.RGB
   EyeDirectionWorld=@n16
   IrisUVRadius=@n2
  ) {materialFunction=/Game/MetaHumans/Common/Face/MaterialFunctions/Eye/MF_EyeRefraction}
  n18 = MaterialFunctionCall(Float3=@n17.IrisMask) {materialFunction=&BreakOutFloat3Components}
  n19 = LinearInterpolate(
   A=ScalarParameter() {defaultValue=0.4, group=Sclera, parameterName=ScleraSpecularity}
   B=ScalarParameter() {defaultValue=0.4, group=Iris, parameterName=IrisSpecularity}
   Alpha=@n18.R
  )
  n20 = LinearInterpolate(
   A=ScalarParameter() {defaultValue=0.1, group=Sclera, parameterName=ScleraRoughness}
   B=ScalarParameter() {defaultValue=0.1, group=Iris, parameterName=IrisRoughness}
   Alpha=@n18.R
  )
  n21 = Constant() {r=0.5}
  n22 = Divide(
   A=MaterialFunctionCall(Vector 2=Subtract(A=@n17.RefractedUV, B=@n21) {desc=Center Refracted UVs}) {materialFunction=/Engine/Functions/Engine_MaterialFunctions02/Utility/VectorLength}.V2 Length
   B=@n2
  ) {desc=Scale UV distance to match Iris size in UV space}
  n23 = ScalarParameter() {defaultValue=0.11, group=Iris, parameterName=Iris Concavity Scale, sliderMax=4}
  n24 = ScalarParameter() {group=ColorCorrection, parameterName=ColorHueShift}
  n25 = ScalarParameter() {defaultValue=0.28, group=Sclera, parameterName=ScleraRotate}
  n26 = TextureSampleParameter2D(UVs=Rotator(Coordinate=NamedRerouteUsage(), Time=Multiply(A=@n25, B=@n14.A)) {speed=6.2832}) {parameterName=ScleraBaseColor, samplerSource=SSM_Wrap_WorldGroupSettings, texture=&/T_Tentacle_Eye_D}
  n27 = NamedRerouteUsage()
  n28 = ScalarParameter() {defaultValue=0.1, group=Refraction, parameterName=POMCutoff}
  n29 = Add(A=@n28)
  n30 = DotProduct(A=Transform(None=Constant3Vector() {constant=<rgba 0 0 -1 0>}) {transformSourceType=TRANSFORMSOURCE_View}, B=@n16)
  n31 = Saturate(None=Subtract(A=Multiply(A=@n29, B=@n30), B=@n28))
  n32 = ScalarParameter() {defaultValue=0.15, group=Refraction, parameterName=POMHeightRatio}
  n33 = Constant() {r=1}
  n34 = Multiply(A=@n2, B=Constant() {r=2}) {desc=Radius to diameter}
  n35 = Constant() {r=0.5}
  MaterialExpressionAdd_4 = Add(
   A=Multiply(
    A=Divide(A=@n33, B=@n34) {desc="Scale Iris texture coordinates up\r\nby this amount before sampling iris"}
    B=Subtract(A=@n17.RefractedUV, B=@n35)
   ) {desc=Centered Scale}
   B=@n35
  )
  MaterialExpressionScalarParameter_21 = ScalarParameter() {defaultValue=1.0501, group=Iris, parameterName=PupilScale}
  MaterialExpressionScalarParameter_41 = ScalarParameter() {defaultValue=0.5376, group=Iris, parameterName=PupilShiftX, sliderMax=1, sliderMin=-1}
  MaterialExpressionScalarParameter_42 = ScalarParameter() {group=Iris, parameterName=PupilShiftY, sliderMax=1, sliderMin=-1}
  n36 = Custom!(
   UV=@MaterialExpressionAdd_4
   PupilScale=@MaterialExpressionScalarParameter_21
   PupilShiftX=@MaterialExpressionScalarParameter_41
   PupilShiftY=@MaterialExpressionScalarParameter_42
  ) {
   code="// Scale UVs from from unit circle in or out from center\r\n// float2 UV, float PupilScale\r\n\r\nfloat ShiftMask = pow(saturate(2*((distance(0.5f, UV) - 0.45f) / -0.5f)),0.7);\r\nPupilShiftX *= ShiftMask*-0.1f;\r\nPupilShiftY *= ShiftMask*0.1f;\r\nfloat2 UVshifted = UV + float2(PupilShiftX, PupilShiftY);\r\nfloat2 UVcentered = UVshifted - float2(0.5f, 0.5f);\r\nfloat UVlength = length(UVcentered);\r\n// UV on circle at distance 0.5 from the center, in direction of original UV\r\nfloat2 UVmax = normalize(UVcentered)*0.5f;\r\n\r\nfloat2 UVscaled = lerp(UVmax, float2(0.0f, 0.0f), saturate((1.f - UVlength*2.f)*PupilScale)) + float2(0.5f, 0.5f);\r\n\r\nreturn UVscaled;"
   description=ScalePupils
   outputType=CMOT_Float2
   inputs [
    {inputName=UV, input {expression=&self:MaterialExpressionAdd_4}}
    {inputName=PupilScale, input {expression=&self:MaterialExpressionScalarParameter_21}}
    {inputName=PupilShiftX, input {expression=&self:MaterialExpressionScalarParameter_41}}
    {inputName=PupilShiftY, input {expression=&self:MaterialExpressionScalarParameter_42}}
   ]
  }
  n37 = ScalarParameter() {defaultValue=2.0246, group=Iris, parameterName=ScaleIrisWithinMask}
  n38 = TextureSampleParameter2D(UVs=MaterialFunctionCall!(
    Heightmap Texture=TextureObjectParameter(UVs=@n27) {parameterName=IrisHeight, texture=&T_Tentacle_Eye_D}
    TangenSpaceForwardVector=@n15
    Height Ratio=Multiply(A=@n31, B=@n32)
    Min Steps=ScalarParameter() {defaultValue=8, group=Refraction, parameterName=POMStepsMin}
    Max Steps=ScalarParameter() {defaultValue=32, group=Refraction, parameterName=POMStepsMax}
    UVs=MaterialFunctionCall(UVs=@n36, Texture Scale=@n37) {materialFunction=&ScaleUVsByCenter}.UVs
    Heightmap Channel=Constant4Vector() {constant=<rgba 1 0 0 0>}.RGBA
    Reference Plane=ScalarParameter() {defaultValue=0.5, group=Refraction, parameterName=POMReferencePlane}
    -----------------------------
    Specify Manual Texture Size
    Manual Texture Size
    ------------------------------
    "Render Shadows (Occlusion Mapping)"
    Shadow Steps
    Shadow Penumbra
   ) {materialFunction=/Game/MetaHumans/Common/Face/MaterialFunctions/Eye/MF_ParallaxOcclusionMappingEye}.Parallax UVs) {parameterName=BaseColor, texture=&T_Tentacle_Eye_D}
  n39 = Multiply(
   A=Power(
    Base=LinearInterpolate(A=@n26.RGB, B=@n38.RGB, Alpha=@n18.R)
    Exp=ScalarParameter() {defaultValue=1.3, group=Sclera, parameterName=ScleraPower}
   )
   B=ScalarParameter() {defaultValue=1.6231, group=Iris, parameterName=IrisBrightness}
  )
  n40 = Multiply(A=VectorParameter() {group=ColorCorrection, parameterName=BaseColorTInt, defaultValue=<rgba 1 1 1 1>}.RGB, B=@n39)
  n41 = StaticSwitchParameter(
   True=Power(
    Base=MaterialFunctionCall(Hue Shift Percentage=@n24, Texture=@n40) {materialFunction=&HueShift}.Result
    Exp=ScalarParameter() {defaultValue=1, group=ColorCorrection, parameterName=ColorContrast}
   )
   False=@n39
  ) {group=ColorCorrection, parameterName=UseColorCorrection}
  n42 = SetMaterialAttributes!(
   MaterialAttributes
   Base Color=@n41
   Specular=@n19
   Roughness=@n20
   Emissive Color=NamedRerouteUsage()
   Normal
   Custom Data 0=@n18.B
   Custom Data 1=Power(
    Base=Multiply(A=@n22, B=@n23)
    Exp=ScalarParameter() {defaultValue=0.28, group=Iris, parameterName=Iris Concavity Power, sliderMax=2, sliderMin=0.1}
   ) {desc=Adjust distance - Drives concavity}
   Shading Model=ShadingModel()
  ) {attributeSetTypes [
    #1
    #2
    #3
    #4
    #5
    #6
    #7
    #8
   ]}
  n43 = SetMaterialAttributes!(MaterialAttributes, Base Color=@n41, Specular=@n19, Roughness=@n20, Emissive Color=NamedRerouteUsage(), Normal, Shading Model=ShadingModel()) {attributeSetTypes [
    #1
    #2
    #3
    #4
    #5
    #8
   ]}
  n44 = FeatureLevelSwitch(Default=@n42, ES3_1=@n43)
  n45 = SetMaterialAttributes!(
   MaterialAttributes
   Base Color=Constant3Vector() {constant=<rgba 0.0176 0.008 0.0252 1>}
   Roughness=Constant() {r=0.45}
   Normal=MaterialFunctionCall(
    Normal=TextureSample() {samplerType=SAMPLERTYPE_Normal, texture=/LimeRock/Assets/Textures/Noisepack_Detail_N}.RGB
    Flatness=Constant() {r=0.5}
   ) {materialFunction=&FlattenNormal}.Result
   Specular=Constant() {r=0.5}
   Emissive Color=Constant()
  ) {attributeSetTypes [
    #1
    #3
    #5
    #2
    #4
   ]}
  n46 = OneMinus(None=Multiply(A=MaterialFunctionCall() {materialFunction=&LinearGradient}.VGradient))
  n47 = Multiply(A=Time(), B=ScalarParameter() {defaultValue=0.08, group=Blinking, parameterName=BlinkSpeed})
  n48 = ScalarParameter() {group=Blinking, parameterName=BlinkOffset}
  n49 = CurveAtlasRowParameter(CurveTime=Frac(None=Add(A=@n47, B=@n48))) {
   atlas=/Game/Environments/Helios/Props/Brimstone/Brimstone_Brazier_A/Blueprints/CurveAtlas_FN_LightFunctions
   curve=/Game/Blueprints/Lighting/LightFunctions/Curve_FireFlickers
  }
  n50 = Multiply(A=OneMinus(None=@n49.A), B=Multiply(A=ScalarParameter() {defaultValue=1, group=Blinking, parameterName=BlinkIntensity}) {constB=2}) {constB=2}
  n51 = Round(None=Add(A=Subtract(A=@n46), B=Clamp(None=@n50) {maxDefault=2}))
  n52 = ScalarParameter() {group=Blinking, parameterName=Blinking_Close}
  n53 = Saturate(None=Add(A=@n51, B=@n52))
  MaterialFunctionCall!(
   Material=MaterialFunctionCall!(In=BlendMaterialAttributes(A=@n44, B=@n45, Alpha=@n53)) {materialFunction=&/MF_WorldGlow_LimeRock}.Result
   NewOpacityMask=Round(None=MaterialFunctionCall!(MaskEnabled) {materialFunction=&/MF_LImeRock_PortalOpacityMask}.OpacityMask)
   Mask
  ) {materialFunction=/Engine/Functions/MaterialLayerFunctions/MatLayerBlend_OverrideOpacityMask}
  Multiply()
  Multiply(A=@n52, B=@n51)
  n54 = ScalarParameter() {parameterName=Eye_X_Offset}
  n55 = VertexInterpolator(VS=LocalPosition() {localOrigin=InstancePreSkinning}.XY)
  n56 = Round(None=Divide(A=ComponentMask(None=@n55.PS) {r=true}, B=ScalarParameter() {defaultValue=6000, parameterName=Witdh}))
  n57 = LinearInterpolate(A=@n54, B=Multiply(A=@n54) {constB=-1}, Alpha=@n56)
  n58 = ScalarParameter() {parameterName=Eye_Y_Offset}
  n59 = LinearInterpolate(A=@n58, B=Multiply(A=@n58) {constB=-1}, Alpha=@n56)
  NamedRerouteDeclaration(None=Add(A=TextureCoordinate(), B=AppendVector(A=@n57, B=@n59))) {name=BaseCoordinate, nodeColor=<rgba 0.3828 1 0 1>}
  n60 = ScalarParameter() {group=Glow, parameterName=EmissiveAmount}
  NamedRerouteDeclaration(None=Multiply(A=Multiply(A=@n41, B=@n60), B=VectorParameter() {parameterName=EmissiveTint, defaultValue=<rgba 1 1 1 1>}.RGB)) {name=Emissive, nodeColor=<rgba 0.4297 0 1 1>}
  Saturate(None=@n50)
  TangentOutput!(None=@n16)
  TextureCoordinate()
 ]
}
