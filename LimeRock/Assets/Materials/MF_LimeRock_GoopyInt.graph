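# LimeRock material function that blends a glowing goop layer over incoming material attributes, weighted by an Alpha input.
# The goop uses parameterized color, specular and roughness, a tiled water normal map and the Subsurface Profile shading model.
# An inverted Fresnel on the transformed normal masks the emissive color.
# MaterialProxyReplace gives baked proxy materials nearly flattened normal UVs and an added tinted emissive boost.
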MaterialFunction /LimeRock/Assets/Materials/MF_LimeRock_GoopyInt
graph {nodes [
  n1 = FunctionInput() {inputName=Material, inputType=FunctionInput_MaterialAttributes}
  n2 = VectorParameter() {parameterName=GoopColor, defaultValue=<rgba 0.2083 0.0374 0.0587 1>}
  n3 = ScalarParameter() {parameterName=SpecGoop}
  n4 = ScalarParameter() {defaultValue=1, parameterName=RoughnessGoop}
  n5 = VectorParameter() {parameterName=GoopEmissiveColorColor, defaultValue=<rgba 2 0.8042 0.0208 1>}
  n6 = ScalarParameter() {defaultValue=0.04, parameterName=CoreFresEX1}
  n7 = TextureCoordinate() {uTiling=2, vTiling=2}
  n8 = ScalarParameter() {defaultValue=0.4261, parameterName=WaterTiling}
  n9 = TextureSample(UVs=MaterialProxyReplace(Realtime=Multiply(A=@n7, B=@n8), MaterialProxy=Multiply(A=@n7) {constB=0.001})) {samplerType=SAMPLERTYPE_Normal, texture=/Game/Global/Textures/Default/Tileable_Patterns/T_Water2_N}
  n10 = Transform(None=@n9.RGB)
  n11 = Multiply(A=@n5.RGB, B=OneMinus(None=Fresnel(ExponentIn=@n6, Normal=@n10) {baseReflectFraction=0, exponent=0.05}))
  n12 = Constant3Vector() {constant=<rgba 0.2822 0.4468 1.6 1>}
  n13 = MaterialProxyReplace(Realtime=@n11, MaterialProxy=Add(A=@n11, B=Multiply(A=@n5.RGB, B=@n12) {constB=0.6}))
  n14 = ShadingModel() {shadingModel=MSM_SubsurfaceProfile}
  FunctionOutput(None=BlendMaterialAttributes(
    A=GetMaterialAttributes(None=@n1).MaterialAttributes
    B=MakeMaterialAttributes!(
     BaseColor=@n2.RGB
     Metallic
     Specular=@n3
     Roughness=@n4
     Anisotropy
     EmissiveColor=@n13
     Opacity
     OpacityMask
     Normal=@n9.RGB
     Tangent
     WorldPositionOffset
     SubsurfaceColor
     ClearCoat
     ClearCoatRoughness
     AmbientOcclusion
     Refraction
     CustomizedUVs_0
     CustomizedUVs_1
     CustomizedUVs_2
     CustomizedUVs_3
     CustomizedUVs_4
     CustomizedUVs_5
     CustomizedUVs_6
     CustomizedUVs_7
     PixelDepthOffset
     ShadingModel=@n14
     Displacement
    )
    Alpha=FunctionInput() {inputName=Alpha, inputType=FunctionInput_Scalar}
   ))
 ]}
