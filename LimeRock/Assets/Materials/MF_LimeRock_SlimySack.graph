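# Material function that blends a slimy, glowing organic sack look over incoming material attributes by a mask input.
# Tints, animated webbing, drips, thin-film fresnel and detail maps shape the surface.
# Panned noise forms bugs; emissive combines a travelling local-position pulse with bug modulation, alongside sky-lit fake subsurface and switchable normal-driven WPO.
# Quality switches drop webbing and drips on low, while MaterialProxyReplace simplifies emissive for HLOD proxies.
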MaterialFunction /LimeRock/Assets/Materials/MF_LimeRock_SlimySack
graph {nodes [
  n1 = FunctionInput() {inputName=Mat, inputType=FunctionInput_MaterialAttributes}
  n2 = GetMaterialAttributes(None=@n1) {attributeGetTypes [#1, #2, #3]}
  n3 = ScalarParameter() {defaultValue=-0.6042, parameterName=SpecularCrevice}
  n4 = ScalarParameter() {defaultValue=0.642, parameterName=spec}
  n5 = NamedRerouteUsage()
  n6 = Constant() {r=1}
  n7 = Multiply(A=TextureCoordinate() {coordinateIndex=1}, B=ScalarParameter() {defaultValue=20, parameterName=DetaiLTiling}) {constB=6.8949}
  n8 = TextureSample(UVs=@n7) {samplerType=SAMPLERTYPE_Normal, texture=/LimeRock/Assets/Textures/Noisepack_Detail_N}
  n9 = ScalarParameter() {defaultValue=0.85, parameterName=DetailNormalIntensity}
  n10 = VertexNormalWS()
  n11 = Constant() {r=-0.064}
  n12 = NamedRerouteUsage()
  n13 = Multiply(
   A=ScalarParameter() {defaultValue=129.5394, group=Bugs, parameterName=BugPopIntensity}
   B=Saturate(None=LinearInterpolate(A=@n11, Alpha=@n12) {constA=-0.5})
  )
  n14 = ScalarParameter() {defaultValue=0.1, parameterName=DripIntensity}
  n15 = NamedRerouteUsage()
  n16 = LinearInterpolate(
   A=VectorParameter() {parameterName=Tint, defaultValue=<rgba 0.13 0.0518 0.1376 1>}.RGB
   B=VectorParameter() {parameterName=Tint2, defaultValue=<rgba 0.4561 0.0648 0.0736 1>}.RGB
   Alpha=NamedRerouteUsage()
  )
  n17 = VectorParameter() {parameterName=WebColor, defaultValue=<rgba 0.1118 0.0591 0.1146 1>}
  n18 = QualitySwitch(Default=LinearInterpolate(A=@n16, B=Multiply(A=@n16, B=@n17.RGB), Alpha=NamedRerouteUsage()), Low=@n16) {desc=Taking out webbing on mobile}
  n19 = Fresnel(ExponentIn=ScalarParameter() {defaultValue=0.348, parameterName=FresEX}) {baseReflectFraction=0, exponent=1}
  n20 = Multiply(
   A=@n19
   B=TextureSample(UVs=@n19) {texture=/Game/Accessories/FORT_Backpacks/Backpack_M_MED_DiscoBall/Textures/T_DiscoBall_Backpack_ThinFilm}.RGB
  )
  n21 = ScalarParameter() {defaultValue=0.9063, parameterName=FresnelBrightness}
  n22 = LinearInterpolate(A=@n18, B=Multiply(A=@n18, B=Multiply(A=@n20, B=@n21)), Alpha=@n19)
  n23 = QualitySwitch(
   Default=LinearInterpolate(
    A=@n22
    B=VectorParameter() {parameterName=DripTint, defaultValue=<rgba 0.0632 0.1094 0.0885 1>}.RGB
    Alpha=LinearInterpolate(B=@n14, Alpha=@n15) {constB=0.6}
   )
   Low=@n22
  ) {desc=Taking out drips on mobile}
  n24 = NamedRerouteUsage()
  n25 = TextureSample(UVs=@n7) {texture=/LimeRock/Assets/Textures/Noisepack_Detail_C}
  n26 = Constant() {r=0.75}
  n27 = NamedRerouteUsage()
  n28 = Add(A=@n27, B=NamedRerouteUsage())
  n29 = SetMaterialAttributes!(
   MaterialAttributes=@n2.MaterialAttributes
   ShadingModel=ShadingModel()
   BaseColor=Multiply(A=Multiply(A=@n23, B=@n24), B=Multiply(A=@n25.RGB, B=@n26) {constB=1.05})
   Roughness=NamedRerouteUsage()
   Specular=Saturate(None=LinearInterpolate(A=@n3, B=@n4, Alpha=@n5))
   SubsurfaceColor
   Opacity=@n6
   Normal=MaterialFunctionCall(
    BaseNormal=@n2.Normal
    AdditionalNormal=MaterialFunctionCall(Normal=@n8.RGB, Flatness=@n9) {materialFunction=&FlattenNormal}.Result
   ) {materialFunction=&BlendAngleCorrectedNormals}.Result
   EmissiveColor=@n28
   CustomData0=Constant() {r=1}
   CustomData1=@n6
   WorldPositionOffset=StaticSwitchParameter(True=Multiply(A=@n10, B=@n13), False=Constant()) {defaultValue=true, parameterName=useWPO}
   PixelDepthOffset
  ) {attributeSetTypes [
    #4
    #1
    #2
    #5
    #6
    #7
    #3
    #8
    #9
    #10
    #11
    #12
   ]}
  FunctionOutput(None=BlendMaterialAttributes(A=@n1, B=@n29, Alpha=FunctionInput() {inputName=Mask, inputType=FunctionInput_Scalar}))
  n30 = ScalarParameter() {group=Bugs, parameterName=Mind}
  n31 = NamedRerouteUsage()
  n32 = ScalarParameter() {defaultValue=0.5, group=Bugs, parameterName=BugFadePower}
  n33 = Saturate(None=Multiply(A=Power(Base=@n31, Exp=@n32) {constExponent=0.5}, B=ScalarParameter() {defaultValue=8.4465, group=Bugs, parameterName=BugEmissiveIntensity}))
  LinearInterpolate(B=@n30, Alpha=@n33) {constA=1, constB=0}
  MaterialProxyReplace(Realtime=@n28, MaterialProxy=@n27)
  n34 = Multiply(A=TextureCoordinate() {coordinateIndex=1}, B=ScalarParameter() {defaultValue=1, group=Bugs, parameterName=ScaleUV})
  n35 = VectorParameter() {group=Bugs, parameterName=BugSpeedPanning, defaultValue=<rgba 0.005 0.05 0 1>}
  n36 = Panner(Coordinate=@n34, Speed=AppendVector(A=Multiply(A=@n35.R) {constB=-1}, B=@n35.B)) {speedX=-0.002, speedY=0.01}
  n37 = Panner(Coordinate=@n34, Speed=AppendVector(A=@n35.R, B=@n35.R)) {speedY=0.03}
  n38 = NamedRerouteUsage()
  n39 = MaterialFunctionCall(Float3=Multiply(A=@n38, B=VectorParameter() {group=Bugs, parameterName=warp, defaultValue=<rgba 0 0.05 0.0208 1>}.RGB)) {materialFunction=&BreakOutFloat3Components}
  n40 = Multiply(A=Add(A=@n34, B=AppendVector(A=@n39.R, B=@n39.R))) {constB=2}
  n41 = Multiply(A=AppendVector(A=@n35.R, B=@n35.R)) {constB=1.45}
  n42 = Max(
   A=Max(
    A=TextureSample(UVs=@n36) {samplerType=SAMPLERTYPE_Masks, texture=/LimeRock/Assets/Textures/Noisepack_Detail_R}.G
    B=TextureSample(UVs=@n37) {samplerType=SAMPLERTYPE_Masks, texture=&Noisepack_Detail_R}.B
   )
   B=TextureSample(UVs=Panner(Coordinate=@n40, Speed=@n41) {speedX=-0.008, speedY=-0.018}) {samplerType=SAMPLERTYPE_Masks, texture=&Noisepack_Detail_R}.B
  )
  n43 = Multiply(A=Multiply(A=@n42, B=@n38), B=NamedRerouteUsage())
  Multiply(A=@n43, B=@n43)
  Multiply(A=NamedRerouteUsage(), B=@n42)
  n44 = Multiply(A=TextureCoordinate() {coordinateIndex=1, uTiling=2.4, vTiling=2.4}, B=ScalarParameter() {defaultValue=1, parameterName=Web_Tiling})
  n45 = TextureSample(UVs=Panner(Coordinate=Multiply(A=@n44) {constB=0.75}) {speedY=-0.033}) {samplerType=SAMPLERTYPE_LinearColor, texture=/LimeRock/Assets/Textures/Noisepack_M}
  n46 = Sine(None=Multiply(A=Time()) {constB=0.1})
  n47 = Multiply(A=@n45.G, B=LinearInterpolate(Alpha=Abs(None=@n46)) {constA=0.3})
  n48 = TextureSample(UVs=Panner(Coordinate=Multiply(A=@n44) {constB=0.7}) {speedX=0.03, speedY=0.02}) {samplerType=SAMPLERTYPE_LinearColor, texture=&Noisepack_M}
  n49 = Min(A=@n45.G, B=@n48.G)
  n50 = TextureSample(UVs=Panner(Coordinate=@n44) {speedX=-0.02, speedY=0.011}) {samplerType=SAMPLERTYPE_LinearColor, texture=&Noisepack_M}
  n51 = ScalarParameter() {defaultValue=0.0781, parameterName=asdfsd}
  n52 = TextureSample(UVs=Add(A=@n44, B=Multiply(A=@n50.B, B=@n51) {constB=0.2})) {samplerType=SAMPLERTYPE_LinearColor, texture=&Noisepack_M}
  n53 = Min(A=@n50.G, B=@n52.G)
  n54 = Min(A=@n50.B, B=@n52.B)
  n55 = Multiply(A=Min(A=Min(A=@n47, B=@n49), B=Min(A=@n53, B=@n54)), B=ScalarParameter() {defaultValue=2.5, parameterName=Bias})
  n56 = TextureSample(UVs=Multiply(A=@n44, B=ScalarParameter() {defaultValue=1.9157, parameterName=WebSecondary_Tiling}) {constB=3}) {samplerType=SAMPLERTYPE_LinearColor, texture=&Noisepack_M}
  n57 = ScalarParameter() {defaultValue=0.8299, parameterName=SecondLayer Webbin}
  n58 = Multiply(A=Power(Base=NamedRerouteUsage()), B=Max(A=Saturate(None=@n55), B=Multiply(A=@n56.B, B=@n57) {constB=0.5}))
  NamedRerouteDeclaration(None=QualitySwitch(Default=Clamp(None=@n58), Low=Constant())) {name=Webbing, nodeColor=<rgba 0 1 0.4375 1>}
  n59 = MaterialProxyReplace(
   Realtime=VectorParameter() {parameterName=EmissiveGlowColor, defaultValue=<rgba 12 2.3654 1.2574 1>}.RGB
   MaterialProxy=Multiply(
    A=VectorParameter() {parameterName=HLOD_EmissiveGlowColor, defaultValue=<rgba 8.874 3.582 2.9734 1>}.RGB
    B=ScalarParameter() {defaultValue=0.8, parameterName=HLODEmissiveIntensity}
   )
  )
  n60 = Fresnel(ExponentIn=ScalarParameter() {defaultValue=0.079, parameterName=GlowPulseEXP}) {baseReflectFraction=0}
  n61 = MaterialFunctionCall(Float3=FunctionInput() {inputName=Localpos}) {materialFunction=&BreakOutFloat3Components}
  n62 = Constant() {r=17000}
  n63 = ScalarParameter() {defaultValue=0.7, parameterName=PulseRadius}
  n64 = Subtract(B=@n63) {constA=0}
  n65 = Add(B=@n63) {constA=1}
  n66 = Time()
  n67 = ScalarParameter() {defaultValue=0.1, parameterName=PulseSpeed}
  n68 = Frac(None=Add(A=ScalarParameter() {parameterName=PulseManualOffset}, B=Multiply(A=@n66, B=@n67)))
  n69 = LinearInterpolate(
   A=ScalarParameter() {defaultValue=0.2, parameterName=MinEmissive}
   Alpha=SphereMask(A=Divide(A=@n61.B, B=@n62) {constB=1000}, B=LinearInterpolate(A=@n64, B=@n65, Alpha=@n68), Radius=@n63) {attenuationRadius=0.25, hardnessPercent=0}
  ) {constA=0.2}
  n70 = MaterialProxyReplace(Realtime=Multiply(A=OneMinus(None=@n60), B=@n69), MaterialProxy=@n69)
  n71 = Saturate(None=Multiply(A=NamedRerouteUsage(), B=ScalarParameter() {defaultValue=0.015, parameterName=Webbing EmissiveSubtraction}))
  n72 = Saturate(None=QualitySwitch(Default=Subtract(A=@n70, B=@n71), Low=@n70) {desc=Taking out webbing on mobile})
  n73 = NamedRerouteUsage()
  n74 = ScalarParameter() {defaultValue=0.015, parameterName=Drip EmissiveSubtraction}
  n75 = QualitySwitch(Default=Subtract(A=@n72, B=Multiply(A=@n73, B=@n74)), Low=@n72) {desc=Taking out Drips on mobile}
  n76 = OneMinus(None=@n33)
  n77 = Saturate(None=Clamp(None=@n76, Min=@n30))
  n78 = MaterialProxyReplace(
   Realtime=StaticSwitchParameter(True=Multiply(A=@n75, B=@n77), False=@n75) {defaultValue=true, group=Bugs, parameterName=Activate Emissive Bugs}
   MaterialProxy=@n75
  )
  NamedRerouteDeclaration(None=Multiply(A=NamedRerouteUsage(), B=Multiply(A=@n59, B=@n78))) {name=Emissive, nodeColor=<rgba 0 0.4609 1 1>}
  n79 = Multiply(A=Max(A=Multiply(A=@n45.R, B=@n48.R), B=Multiply(B=@n52.R) {constA=0.4, constB=0.8}), B=@n50.R)
  NamedRerouteDeclaration(None=@n79) {name=PerlinRandomPanner, nodeColor=<rgba 0.9922 0 1 1>}
  NamedRerouteDeclaration(None=Multiply(A=@n43, B=Constant() {r=2})) {name=Bug, nodeColor=<rgba 1 0.9844 0 1>}
  n80 = Abs(None=Subtract(A=TextureCoordinate() {coordinateIndex=1}, B=Constant2Vector() {g=0.5, r=0.5}))
  n81 = ComponentMask(None=@n80) {r=true}
  n82 = ScalarParameter() {defaultValue=0.25, parameterName=Rad}
  n83 = ComponentMask(None=@n80) {g=true}
  n84 = ScalarParameter() {defaultValue=0.402, parameterName=Rad2}
  n85 = OneMinus(None=Max(A=Divide(A=@n81, B=@n82), B=Divide(A=@n83, B=@n84)))
  n86 = Constant() {r=4}
  NamedRerouteDeclaration(None=Saturate(None=Multiply(A=@n85, B=@n86))) {name=UVEdgeProtection, nodeColor=<rgba 1 0 0.9375 1>}
  n87 = ScalarParameter() {parameterName=Bug SubsurfaceReduction}
  n88 = NamedRerouteUsage()
  n89 = Multiply(
   A=VectorParameter() {parameterName=Subsurfaceintensity_Sack, defaultValue=<rgba 0.2812 0.0818 0.096 1>}.RGB
   B=Multiply(A=SkyAtmosphereLightIlluminanceOnGround(), B=ScalarParameter() {defaultValue=0.082, parameterName=SkyContribution})
  )
  n90 = NamedRerouteUsage()
  n91 = ScalarParameter() {defaultValue=0.1163, parameterName=Webbing_SSS}
  n92 = NamedRerouteUsage()
  n93 = QualitySwitch(Default=Multiply(A=@n90, B=LinearInterpolate(B=@n91, Alpha=@n92) {constA=1, constB=0.2}), Low=@n90) {desc=Taking out webbing on mobile}
  NamedRerouteDeclaration(None=Multiply(A=LinearInterpolate(B=@n87, Alpha=@n88) {constA=1, constB=0.7}, B=Multiply(A=@n89, B=@n93))) {name=Fake SSS, nodeColor=<rgba 0.1719 1 0 1>}
  n94 = TextureSample(UVs=TextureCoordinate() {coordinateIndex=1}) {samplerType=SAMPLERTYPE_Masks, texture=&Noisepack_Detail_R}
  NamedRerouteDeclaration(None=@n94.R) {name=Drips, nodeColor=<rgba 1 0.0938 0 1>}
  n95 = ScalarParameter() {defaultValue=1.3266, parameterName=OffSide_SSS_Intensity}
  n96 = ScalarParameter() {defaultValue=0.4009, parameterName=SunSide_SSS_Intensity}
  n97 = SphereMask(A=SkyAtmosphereLightDirection(), B=PixelNormalWS(), Radius=ScalarParameter() {defaultValue=9.8079, parameterName=UnLightRadius}) {hardnessPercent=0}
  NamedRerouteDeclaration(None=Saturate(None=LinearInterpolate(A=@n95, B=@n96, Alpha=@n97) {constA=0.25})) {name=Subsurface, nodeColor=<rgba 0.6406 1 0 1>}
  NamedRerouteDeclaration(None=@n2.Roughness) {name=Roughness, nodeColor=<rgba 0.2891 0 1 1>}
  n98 = MaterialFunctionCall(Float3=FunctionInput() {inputName=M}) {materialFunction=&BreakOutFloat3Components}
  NamedRerouteDeclaration(None=@n98.R) {name=AO, nodeColor=<rgba 0 1 0.0859 1>}
  NamedRerouteDeclaration(None=Power(Base=@n98.G) {constExponent=3}) {name=Gmap, nodeColor=<rgba 1 0 0.3516 1>}
  n99 = NamedRerouteUsage()
  n100 = Multiply(A=@n99, B=ScalarParameter() {defaultValue=0.3, parameterName=Roughness})
  n101 = LinearInterpolate(A=@n100, B=Multiply(A=@n99, B=ScalarParameter() {defaultValue=1, parameterName=Roughness_webbing}), Alpha=NamedRerouteUsage())
  NamedRerouteDeclaration(None=LinearInterpolate(
    A=QualitySwitch(Default=@n101, Low=@n100) {desc=Taking out webbing on mobile}
    B=ScalarParameter() {defaultValue=0.5, parameterName=Roughness_Grime}
    Alpha=@n94.R
   )) {name=Roughness Out, nodeColor=<rgba 0 1 0.6016 1>}
  NamedRerouteUsage()
  NamedRerouteUsage()
  NamedRerouteUsage()
  SmoothStep(Value=@n76)
  StaticSwitch(True=@n79, False=Constant()) {defaultValue=true}
  TextureCoordinate() {coordinateIndex=1}
  TextureSample() {texture=/Game/Packages/Fortress_SharedMaterials/Textures/T_UV_Checker_D}
 ]}
