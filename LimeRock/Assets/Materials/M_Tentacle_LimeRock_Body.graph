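# Skeletal mesh material for the LimeRock tentacle body, built between the 2023 CharacterParent PreFX and PostFX functions.
# Tech Art Mask channels drive stacked slimy sack, leathery skin and pulsing vein layers.
# A static switch then selects goopy interior or world glow.
# MPC_LimeRock's TentacleEmissive blends emissive from Fresnel-weighted to full, and a rounded portal function mask overrides opacity mask for hard clipping.
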Material /LimeRock/Assets/Materials/M_Tentacle_LimeRock_Body
graph {
 properties {
  bAutomaticallySetUsageInEditor=false
  bUseMaterialAttributes=true
  bUsedWithMeshDeformer=true
  bUsedWithSkeletalMesh=true
  opacityMaskClipValue=0.5
  shadingModel=MSM_FromMaterialExpression
  subsurfaceProfile=&/SSP_SlimySack
  usedShadingModels=Default Lit | Subsurface | Clear Coat | Subsurface Profile | Cloth
 }
 nodes [
  n1 = MaterialFunctionCall() {materialFunction=&MF_CharacterParent_2023_PreFX}
  ChannelMaskParameter(None=@n1.Tech Art Mask) {group=Masks, maskChannel=Alpha, parameterName=GoopyMask, defaultValue=<rgba 0 0 0 1>}
  n2 = ChannelMaskParameter(None=@n1.Tech Art Mask) {group=Masks, parameterName=VeinMask}
  n3 = ChannelMaskParameter(None=@n1.Tech Art Mask) {group=Masks, maskChannel=Green, parameterName=LeatherySkinMask, defaultValue=<rgba 0 1 0 0>}
  n4 = ChannelMaskParameter(None=@n1.Tech Art Mask) {group=Masks, maskChannel=Blue, parameterName=SlimySackMaskColor, defaultValue=<rgba 0 0 1 0>}
  n5 = MaterialFunctionCall!(
   Mask=@n2.R
   Mat=MaterialFunctionCall!(
    Alpha=Max(A=@n3.R, B=@n2.R)
    M=@n1.M
    Material=MaterialFunctionCall!(Mask=@n4.R, M=@n1.M, Localpos=@n1.M, Mat=@n1.M) {materialFunction=&/MF_LimeRock_SlimySack}.Result
   ) {materialFunction=&/MF_LimeRock_LeatherySkin}.Result
  ) {materialFunction=&/MF_LimeRock_PulsingVeins}
  n6 = MaterialFunctionCall!(Alpha=@n1.Tech Art Mask Alpha, Material=@n5.Result) {materialFunction=&/MF_LimeRock_GoopyInt}
  n7 = MaterialFunctionCall!(In=@n5.Result) {materialFunction=&/MF_WorldGlow_LimeRock}
  n8 = StaticBoolParameter() {parameterName=UseGoopuInt}
  n9 = GetMaterialAttributes(None=MaterialFunctionCall(
    Material=StaticSwitch(True=@n6.Result, False=@n7.Result, Value=@n8)
    M=@n1.M
    Pre Skinned Local Position=@n1.M
    Distance Blend=@n1.M
    Fresnel=@n1.M
   ) {materialFunction=&MF_CharacterParent_2023_PostFX}.Result) {attributeGetTypes [#1]}
  n10 = Multiply(A=@n9.Emissive Color, B=Fresnel() {baseReflectFraction=0, exponent=1.6}) {constB=0}
  n11 = CollectionParameter() {collection=&/MPC_LimeRock, parameterName=TentacleEmissive}
  MaterialFunctionCall!(
   Material=SetMaterialAttributes(MaterialAttributes=@n9.MaterialAttributes, Emissive Color=LinearInterpolate(A=@n10, B=@n9.Emissive Color, Alpha=@n11)) {attributeSetTypes [#1]}
   NewOpacityMask=Round(None=MaterialFunctionCall!(MaskEnabled) {materialFunction=&/MF_LImeRock_PortalOpacityMask}.OpacityMask)
   Mask
  ) {materialFunction=/Engine/Functions/MaterialLayerFunctions/MatLayerBlend_OverrideOpacityMask}
  Reroute(None=@n1.Result)
 ]
}
